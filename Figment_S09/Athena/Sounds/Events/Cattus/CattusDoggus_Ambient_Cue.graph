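# Ambient SoundCue for the Cattus/Doggus event, mixing two looping ambient waves.
# Continuous modulators driven by a "Progress" parameter crossfade from Loop_02 to Loop_04, with the first loop dropping in pitch as it fades out.
# Attenuation is overridden with a custom falloff curve, distance low-pass filtering and listener focus, and playback uses an extra-high-priority event sound class.
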SoundCue /Figment_S09/Athena/Sounds/Events/Cattus/CattusDoggus_Ambient_Cue
root=Mixer_0
nodes [
 {id=Mixer_0, children [ModulatorContinuous_0, ModulatorContinuous_2], properties {inputVolume [1,1]}}
 {
  id=ModulatorContinuous_0
  children [WavePlayer_1]
  properties {
   pitchModulationParams {default=0, maxOutput=0.8, minOutput=1, parameterName=Progress}
   volumeModulationParams {default=0, maxOutput=0, minOutput=1, parameterName=Progress}
  }
 }
 {
  id=ModulatorContinuous_2
  children [WavePlayer_2]
  properties {pitchModulationParams {default=0, minOutput=1, parameterName=Progress}, volumeModulationParams {default=0, parameterName=Progress}}
 }
 {id=WavePlayer_1, properties {bLooping=true, soundWaveAssetPtr=&/DoggusCattus_Ambient_Loop_02}}
 {id=WavePlayer_2, properties {bLooping=true, soundWaveAssetPtr=&/DoggusCattus_Ambient_Loop_04}}
]
properties {
 bOverrideAttenuation=true
 maxDistance=7524
 soundClassObject=/Figment_S09/Athena/Sounds/Events/SoundClasses/Fort_Events_Priority_ExtHigh
 volumeMultiplier=2
 attenuationOverrides {
  bAttenuateWithLPF=true
  bEnableFocusInterpolation=true
  bEnableListenerFocus=true
  distanceAlgorithm=Custom
  falloffDistance=6500
  focusAttackInterpSpeed=0.5
  focusAzimuth=15
  focusDistanceScale=0.5
  focusReleaseInterpSpeed=0.25
  lPFFrequencyAtMax=500
  lPFRadiusMax=2000
  lPFRadiusMin=1000
  nonFocusAzimuth=90
  nonFocusVolumeAttenuation=0.1
  stereoSpread=2048
  attenuationShapeExtents {x=1024}
  customAttenuationCurve {editorCurveData {keys [<at 0 1>,<at 0.3 1>,<at 0.4 0.5>,<at 1 0>]}}
 }
}
