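# CattusDoggus_Cattus_Impact_05_Cue is a Sound Cue named as an impact sound for Cattus in the CattusDoggus event.
# Its node graph is empty with no root node, so it references no wave and produces no audio as stored here.
# It keeps only playback settings: a map-wide event attenuation with a very large max distance, an important in-game mission sound class, and reduced volume.
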SoundCue /Figment_S09/Athena/Sounds/Events/Cattus/CattusDoggus_Cattus_Impact_05_Cue
root=null
nodes []
properties {
 attenuationSettings=/Figment_S09/Athena/Sounds/Events/Attenuations/Events_Mapwide_Att
 maxDistance=2097152
 soundClassObject=&Fort_Mission_Ingame_Important_PSM
 volumeMultiplier=0.4
}
