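# Sound cue that plays the Cattus Doggus event music stem as non-attenuated, non-spatialized audio.
# A quality level node at the root routes the first two quality slots to the Jammus v1_01 wave and the last two to its "hack" variant.
# The cue uses an important in-game mission sound class, a reduced volume and no reverb send.
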SoundCue /Figment_S09/Athena/Sounds/Events/Cattus/CattusDoggus_Music_Stem
root=QualityLevel_0
nodes [
 {id=QualityLevel_0, children [WavePlayer_1,WavePlayer_1,WavePlayer_0,WavePlayer_0]}
 {id=WavePlayer_0, properties {soundWaveAssetPtr=/Figment_S09/Athena/Sounds/Events/Cattus/Cattus_Doggus/Cattus_Doggus_Jammus_v1_01_hack}}
 {id=WavePlayer_1, properties {soundWaveAssetPtr=/Figment_S09/Athena/Sounds/Events/Cattus/Cattus_Doggus/Cattus_Doggus_Jammus_v1_01}}
]
properties {
 maxDistance=2097152
 soundClassObject=&Fort_Mission_Ingame_Important_PSM
 volumeMultiplier=0.6
 attenuationOverrides {bAttenuate=false, bEnableReverbSend=false, bSpatialize=false}
}
