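# Sound cue for the Cattus event that plays a "Down" sound, with its exact in-game trigger not shown.
# A Random node picks one of three equally weighted Cattus_Down wave variations.
# Playback uses the Cattus_Focus_Att attenuation asset with a very large max distance and routes to the medium-priority events sound class.
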SoundCue /Figment_S09/Athena/Sounds/Events/Cattus/Cattus_Down_Cue
root=Random_0
nodes [
 {id=Random_0, children [WavePlayer_0,WavePlayer_1,WavePlayer_2], properties {weights [1,1,1]}}
 {id=WavePlayer_0, properties {soundWaveAssetPtr=&/Cattus_Down_03}}
 {id=WavePlayer_1, properties {soundWaveAssetPtr=&/Cattus_Down_01}}
 {id=WavePlayer_2, properties {soundWaveAssetPtr=&/Cattus_Down_02}}
]
properties {
 attenuationSettings=&/Cattus_Focus_Att
 maxDistance=1000000
 soundClassObject=/Figment_S09/Athena/Sounds/Events/SoundClasses/Fort_Events_Priority_Medium
}
