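# Sound cue for the Cattus event's "Exist" sound, likely an idle or presence vocalization of the Cattus creature.
# A random node picks one of three equally weighted Cattus_Exist wave variations on each play.
# Attenuation overrides give it very long-range falloff, wide stereo spread, distance-scaled reverb and interpolated focus, and it routes through the medium-priority events sound class.
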SoundCue /Figment_S09/Athena/Sounds/Events/Cattus/Cattus_Exist_Cue
root=Random_0
nodes [
 {id=Random_0, children [WavePlayer_0,WavePlayer_1,WavePlayer_2], properties {weights [1,1,1]}}
 {id=WavePlayer_0, properties {soundWaveAssetPtr=&/Cattus_Exist_03}}
 {id=WavePlayer_1, properties {soundWaveAssetPtr=&/Cattus_Exist_01}}
 {id=WavePlayer_2, properties {soundWaveAssetPtr=&/Cattus_Exist_02}}
]
properties {
 bOverrideAttenuation=true
 maxDistance=20000
 soundClassObject=/Figment_S09/Athena/Sounds/Events/SoundClasses/Fort_Events_Priority_Medium
 volumeMultiplier=1
 attenuationOverrides {
  bEnableFocusInterpolation=true
  falloffDistance=12500
  focusAttackInterpSpeed=1.5
  reverbDistanceMax=15000
  reverbDistanceMin=5000
  reverbWetLevelMax=10
  reverbWetLevelMin=0
  stereoSpread=2048
  attenuationShapeExtents {x=7500}
 }
}
