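# Looping sound cue for the Cattus event, built from traverse loop sounds.
# A Random node picks one of three equally weighted chopped variations, each played by a looping wave player.
# It uses a Cattus focus-loop attenuation asset with a very large max distance and routes to the medium-priority events sound class.
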SoundCue /Figment_S09/Athena/Sounds/Events/Cattus/Cattus_Loop_Cue
root=Random_1
nodes [
 {id=Random_1, children [WavePlayer_4,WavePlayer_5,WavePlayer_6], properties {weights [1,1,1]}}
 {id=WavePlayer_4, properties {bLooping=true, soundWaveAssetPtr=&/Cattus_Traverse_Loop_01_chop_1}}
 {id=WavePlayer_5, properties {bLooping=true, soundWaveAssetPtr=&/Cattus_Traverse_Loop_01_chop_2}}
 {id=WavePlayer_6, properties {bLooping=true, soundWaveAssetPtr=&/Cattus_Traverse_Loop_01_chop_3}}
]
properties {
 attenuationSettings=&/Cattus_Focus_Loop_Att
 maxDistance=1000000
 soundClassObject=/Figment_S09/Athena/Sounds/Events/SoundClasses/Fort_Events_Priority_Medium
}
