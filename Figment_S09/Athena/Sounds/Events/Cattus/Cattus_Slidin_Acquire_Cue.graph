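# Sound cue for the Cattus "Slidin Acquire" event sound, routed to the medium-priority events sound class.
# A Random node picks one of four equally weighted Slidin wave variations and feeds it through a root Modulator left at default settings.
# Spatialization references the shared Cattus_Slidin attenuation asset.
# The cue also stores attenuation override values for listener focus, distance falloff, wide stereo spread and reverb send, without showing the override switched on.
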SoundCue /Figment_S09/Athena/Sounds/Events/Cattus/Cattus_Slidin_Acquire_Cue
root=Modulator_0
nodes [
 {id=Modulator_0, children [Random_1]}
 {id=Random_1, children [WavePlayer_3,WavePlayer_4,WavePlayer_5,WavePlayer_6], properties {weights [1,1,1,1]}}
 {id=WavePlayer_3, properties {soundWaveAssetPtr=&/Cattus_Slidin_04}}
 {id=WavePlayer_4, properties {soundWaveAssetPtr=&/Cattus_Slidin_03}}
 {id=WavePlayer_5, properties {soundWaveAssetPtr=&/Cattus_Slidin_02}}
 {id=WavePlayer_6, properties {soundWaveAssetPtr=&/Cattus_Slidin_01}}
]
properties {
 attenuationSettings=&/Cattus_Slidin_Att
 maxDistance=7500
 soundClassObject=/Figment_S09/Athena/Sounds/Events/SoundClasses/Fort_Events_Priority_Medium
 volumeMultiplier=1
 attenuationOverrides {
  bEnableFocusInterpolation=true
  bEnableListenerFocus=true
  falloffDistance=12500
  focusAttackInterpSpeed=1.5
  reverbDistanceMax=15000
  reverbDistanceMin=5000
  reverbWetLevelMax=10
  reverbWetLevelMin=0
  stereoSpread=2048
  attenuationShapeExtents {x=7500}
 }
}
