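# Looping sound cue for the Cattus sliding effect in a Figment_S09 event, spatialized with attenuation on a medium-priority events sound class.
# A mixer layers a slide loop with a switched ice-castle ambience loop.
# The SquishDelta parameter continuously drives the slide loop's volume and pitch.
# The ambience loop is pitched down, plays only when the Active parameter is 1, and swells from quiet to full volume over five seconds.
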SoundCue /Figment_S09/Athena/Sounds/Events/Cattus/Cattus_Slidin_Loop_Cue
root=Mixer_2
nodes [
 {
  id=Enveloper_0
  children [Modulator_0]
  properties {pitchCurve {editorCurveData {keys [{time=0}]}}, volumeCurve {editorCurveData {keys [<at 0 0.1>,<at 5 1>]}}}
 }
 {id=Mixer_2, children [ModulatorContinuous_1, Switch_0], properties {inputVolume [1,1]}}
 {
  id=ModulatorContinuous_1
  children [WavePlayer_1]
  properties {
   pitchModulationParams {default=0, maxOutput=1.2, minOutput=0.8, parameterName=SquishDelta}
   volumeModulationParams {default=0, maxInput=0.5, parameterName=SquishDelta}
  }
 }
 {id=Modulator_0, children [WavePlayer_0], properties {pitchMax=0.75, pitchMin=0.75}}
 {id=Switch_0, children [null,null,Enveloper_0], properties {intParameterName=Active}}
 {id=WavePlayer_0, properties {bLooping=true, soundWaveAssetPtr=/Figment_S09/Athena/Sounds/Ambient/IceCastle/IceCastle_Egg_01}}
 {id=WavePlayer_1, properties {bLooping=true, soundWaveAssetPtr=&/Cattus_Slidin_Loop_02}}
]
properties {
 attenuationSettings=&/Cattus_Slidin_Small_Att
 maxDistance=4000
 soundClassObject=/Figment_S09/Athena/Sounds/Events/SoundClasses/Fort_Events_Priority_Medium
 volumeMultiplier=1.75
 attenuationOverrides {
  bEnableFocusInterpolation=true
  bEnableListenerFocus=true
  falloffDistance=12500
  focusAttackInterpSpeed=1.5
  reverbDistanceMax=15000
  reverbDistanceMin=5000
  reverbWetLevelMax=10
  reverbWetLevelMin=0
  stereoSpread=2048
  attenuationShapeExtents {x=7500}
 }
}
