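# Sound cue for the Cattus "Up" sound in the Figment S09 events audio, routed to the medium-priority events sound class.
# A random node picks one of three equally weighted Up waves and feeds an attenuation node using the Cattus Focus attenuation, also set on the cue.
# That feeds a mixer with an empty second input, and a root delay holds playback for a fixed 1.5 seconds.
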SoundCue /Figment_S09/Athena/Sounds/Events/Cattus/Cattus_Up_Cue
root=Delay_1
nodes [
 {id=Attenuation_0, children [Random_0], properties {attenuationSettings=&/Cattus_Focus_Att}}
 {id=Delay_1, children [Mixer_0], properties {delayMax=1.5, delayMin=1.5}}
 {id=Mixer_0, children [Attenuation_0, null], properties {inputVolume [1,1]}}
 {id=Random_0, children [WavePlayer_0,WavePlayer_1,WavePlayer_2], properties {weights [1,1,1]}}
 {id=WavePlayer_0, properties {soundWaveAssetPtr=&/Cattus_Up_01}}
 {id=WavePlayer_1, properties {soundWaveAssetPtr=&/Cattus_Up_02}}
 {id=WavePlayer_2, properties {soundWaveAssetPtr=&/Cattus_Up_03}}
]
properties {
 attenuationSettings=&Cattus_Focus_Att
 maxDistance=1000000
 soundClassObject=/Figment_S09/Athena/Sounds/Events/SoundClasses/Fort_Events_Priority_Medium
}
