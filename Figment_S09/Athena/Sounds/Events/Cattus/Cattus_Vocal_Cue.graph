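# Sound cue for Cattus event vocals, playing one of four vocal waves after a fixed four-second delay.
# A Random node with equal weights picks the wave and feeds an Attenuation node, then the Delay at the root.
# The cue uses the Cattus vocal attenuation, a medium-priority events sound class, a boosted volume and a very large max distance.
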SoundCue /Figment_S09/Athena/Sounds/Events/Cattus/Cattus_Vocal_Cue
root=Delay_1
nodes [
 {id=Attenuation_0, children [Random_1], properties {attenuationSettings=&/Cattus_Vocal_Att}}
 {id=Delay_1, children [Attenuation_0], properties {delayMax=4, delayMin=4}}
 {id=Random_1, children [WavePlayer_3,WavePlayer_4,WavePlayer_5,WavePlayer_6], properties {weights [1,1,1,1]}}
 {id=WavePlayer_3, properties {soundWaveAssetPtr=&/Cattus_Vocal_01}}
 {id=WavePlayer_4, properties {soundWaveAssetPtr=&/Cattus_Vocal_02}}
 {id=WavePlayer_5, properties {soundWaveAssetPtr=&/Cattus_Vocal_03}}
 {id=WavePlayer_6, properties {soundWaveAssetPtr=&/Cattus_Vocal_04}}
]
properties {
 attenuationSettings=&Cattus_Vocal_Att
 maxDistance=1000000
 soundClassObject=/Figment_S09/Athena/Sounds/Events/SoundClasses/Fort_Events_Priority_Medium
 volumeMultiplier=1.75
}
